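# First-person fire sound cue for the 1911 pistol (HT), playing one of three equally weighted close-fire waves.
# A random node feeds every branch of a quality-level switch, so all levels sound alike, then a modulator varies pitch slightly upward and boosts volume.
# It uses first-person projectile attenuation, a ranged-fire sound class and the first-person gunfire concurrency set.
# The cue's debug flag is left on.
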SoundCue /Game/Sounds/Fort_Weapons/pistol/1911/1911_HT_Fire_1P_Cue
root=Modulator_2
nodes [
 {id=Modulator_2, children [QualityLevel_0], properties {pitchMax=1.1, pitchMin=1, volumeMax=1.9, volumeMin=1.8}}
 {id=QualityLevel_0, children [Random_2,Random_2,Random_2,Random_2]}
 {id=Random_2, children [WavePlayer_12, WavePlayer_13, WavePlayer_14], properties {weights [1,1,1]}}
 {id=WavePlayer_12, properties {soundWaveAssetPtr=&/1911_Excite_Fire_Close_01}}
 {id=WavePlayer_13, properties {soundWaveAssetPtr=&/1911_Excite_Fire_Close_02}}
 {id=WavePlayer_14, properties {soundWaveAssetPtr=&/1911_Excite_Fire_Close_03}}
]
properties {
 attenuationSettings=&Fort_Weapon_Projectile_FirstPerson_Athena_Att
 bDebug=true
 maxDistance=2097152
 soundClassObject=&Fort_WeaponRanged_Fire_1P_PSM
 volumeMultiplier=0.7
 concurrencyOverrides {bLimitToOwner=true, maxCount=3, resolutionRule=StopOldest}
 concurrencySet [&Weapon_Gunfire_FirstPerson_Conc]
}
