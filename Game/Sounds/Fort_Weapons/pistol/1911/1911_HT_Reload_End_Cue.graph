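# Sound cue for the end of the 1911 pistol (HT variant) reload, played as weapon foley.
# It randomly picks one of two equally weighted reload-end waves, then passes it through a weapon foley sound class node, a modulator for variation, and reload attenuation.
# The cue slightly lowers pitch and volume, overrides concurrency, and has its debug flag on.
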SoundCue /Game/Sounds/Fort_Weapons/pistol/1911/1911_HT_Reload_End_Cue
root=Attenuation_0
nodes [
 {id=Attenuation_0, children [Modulator_0], properties {attenuationSettings=&Fort_Weapon_Reload_Att}}
 {id=Modulator_0, children [SoundClass_0]}
 {id=Random_0, children [WavePlayer_0,WavePlayer_1], properties {weights [1,1]}}
 {id=SoundClass_0, children [Random_0], properties {soundClassOverride=&Fort_WeaponFoley}}
 {id=WavePlayer_0, properties {soundWaveAssetPtr=&/1911_Reload_End_01}}
 {id=WavePlayer_1, properties {soundWaveAssetPtr=&/1911_Reload_End_02}}
]
properties {
 attenuationSettings=&Fort_Weapon_Reload_Att
 bDebug=true
 bOverrideConcurrency=true
 maxDistance=2250
 pitchMultiplier=0.9
 soundClassObject=&Fort_WeaponFoley
 volumeMultiplier=0.902
}
