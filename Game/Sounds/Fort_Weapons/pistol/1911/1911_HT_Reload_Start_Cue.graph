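# Sound cue for the start of the 1911 HT pistol reload sound in Fortnite weapon audio.
# A single wave player feeds a weapon foley sound class node, then a modulator at default settings, then a reload attenuation node at the root.
# The cue itself also sets the same attenuation and sound class, lowers volume, overrides concurrency and has its debug flag on.
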SoundCue /Game/Sounds/Fort_Weapons/pistol/1911/1911_HT_Reload_Start_Cue
root=Attenuation_0
nodes [
 {id=Attenuation_0, children [Modulator_0], properties {attenuationSettings=&Fort_Weapon_Reload_Att}}
 {id=Modulator_0, children [SoundClass_0]}
 {id=SoundClass_0, children [WavePlayer_1], properties {soundClassOverride=&Fort_WeaponFoley}}
 {id=WavePlayer_1, properties {soundWaveAssetPtr=&/1911_Reload_Start_01}}
]
properties {
 attenuationSettings=&Fort_Weapon_Reload_Att
 bDebug=true
 bOverrideConcurrency=true
 maxDistance=2250
 soundClassObject=&Fort_WeaponFoley
 volumeMultiplier=0.702
}
